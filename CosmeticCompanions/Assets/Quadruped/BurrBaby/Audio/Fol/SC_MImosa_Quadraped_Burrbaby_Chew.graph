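# Sound cue for the BurrBaby quadruped companion pet's chewing foley, routed to the creature pets sound class with companion foley attenuation.
# A quality switch sends its first two levels to a twelve-clip random pool and its last two to a smaller pool, where one zero-weighted clip leaves four active.
# A modulator at the root randomizes pitch and volume.
# The cue uses the shared quadruped footstep concurrency rather than a chew-specific one.
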SoundCue /CosmeticCompanions/Assets/Quadruped/BurrBaby/Audio/Fol/SC_MImosa_Quadraped_Burrbaby_Chew
root=Modulator_0
nodes [
 {id=Modulator_0, children [QualityLevel_0], properties {pitchMax=1.2, pitchMin=0.9, volumeMax=1, volumeMin=0.6}}
 {id=QualityLevel_0, children [Random_2,Random_2,Random_3,Random_3]}
 {
  id=Random_2
  children [
   WavePlayer_0
   WavePlayer_1
   WavePlayer_2
   WavePlayer_3
   WavePlayer_4
   WavePlayer_5
   WavePlayer_6
   WavePlayer_7
   WavePlayer_8
   WavePlayer_9
   WavePlayer_10
   WavePlayer_11
  ]
  properties {weights [1,1,1,1,1,1,1,1,1,1,1,1]}
 }
 {id=Random_3, children [WavePlayer_12, WavePlayer_13, WavePlayer_14, WavePlayer_30, WavePlayer_31], properties {weights [1,1,1,1,0]}}
 {id=WavePlayer_0, properties {soundWaveAssetPtr=&/mimosa_burrbaby_fol_chew_12}}
 {id=WavePlayer_1, properties {soundWaveAssetPtr=&/mimosa_burrbaby_fol_chew_11}}
 {id=WavePlayer_10, properties {soundWaveAssetPtr=&/mimosa_burrbaby_fol_chew_02}}
 {id=WavePlayer_11, properties {soundWaveAssetPtr=&/mimosa_burrbaby_fol_chew_01}}
 {id=WavePlayer_12, properties {soundWaveAssetPtr=&mimosa_burrbaby_fol_chew_12}}
 {id=WavePlayer_13, properties {soundWaveAssetPtr=&mimosa_burrbaby_fol_chew_11}}
 {id=WavePlayer_14, properties {soundWaveAssetPtr=&/mimosa_burrbaby_fol_chew_10}}
 {id=WavePlayer_2, properties {soundWaveAssetPtr=&mimosa_burrbaby_fol_chew_10}}
 {id=WavePlayer_3, properties {soundWaveAssetPtr=&/mimosa_burrbaby_fol_chew_09}}
 {id=WavePlayer_30, properties {soundWaveAssetPtr=&mimosa_burrbaby_fol_chew_09}}
 {id=WavePlayer_31, properties {soundWaveAssetPtr=&/mimosa_burrbaby_fol_chew_03}}
 {id=WavePlayer_4, properties {soundWaveAssetPtr=&/mimosa_burrbaby_fol_chew_08}}
 {id=WavePlayer_5, properties {soundWaveAssetPtr=&/mimosa_burrbaby_fol_chew_07}}
 {id=WavePlayer_6, properties {soundWaveAssetPtr=&/mimosa_burrbaby_fol_chew_06}}
 {id=WavePlayer_7, properties {soundWaveAssetPtr=&/mimosa_burrbaby_fol_chew_05}}
 {id=WavePlayer_8, properties {soundWaveAssetPtr=&/mimosa_burrbaby_fol_chew_04}}
 {id=WavePlayer_9, properties {soundWaveAssetPtr=&mimosa_burrbaby_fol_chew_03}}
]
properties {
 attenuationSettings=/Game/Sounds/Attenuation/AS_Companion_Foley
 maxDistance=1200
 priority=0
 soundClassObject=&Fort_Creatures_Pets
 volumeMultiplier=0.35
 concurrencySet [/CosmeticCompanions/Assets/Audio_Shared/Foley/Quadraped/CONC_Mimosa_Quadraped_Footstep]
}
